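# Physics asset for the MED Croissant cape cosmetic, defining bodies and constraints for cape motion against the character's body.
# Four bone chains (outer and mid, left and right) hang from spine_05 on free-swinging roots, using flat boxes that widen toward the hem with limited swing and twist.
# Free-angle constraints also tie each outer chain to its neighboring mid chain near the top.
# Capsules cover torso, legs and feet; arms have no bodies.
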PhysicsAsset /BRCosmetics/Accessories/FORT_Capes/M_MED_Croissant_Cape/M_MED_Croissant_Cape_Physics
bodies {
 calf_l [{shapeType=Capsule, length=18.4617, radius=11.5383, shapeName=calf_l_capsule, center=<-18.8774 0.0001 0>, rotation=<rot 90 90 90.0002>}]
 calf_r [{shapeType=Capsule, length=21.2592, radius=12.1615, shapeName=calf_r_capsule, center=<21.6024 -0.0001 0>, rotation=<rot -90 -6.7213 6.7211>}]
 cape_02_l [{shapeType=Capsule, length=1, radius=1, shapeName=cape_02_l_capsule}]
 cape_02_r [{shapeType=Capsule, length=1, radius=1, shapeName=cape_02_r_capsule}]
 cape_03_l [{shapeType=Box, extentX=22.5625, extentY=2.6875, extentZ=29.0244, center=<11.5978 0.0001 0>}]
 cape_03_r [{shapeType=Box, extentX=22.5625, extentY=2.6875, extentZ=29.0244, center=<11.5978 0.0001 0>}]
 cape_04_l [{shapeType=Box, extentX=18.0686, extentY=1.4348, extentZ=31.6151, center=<9.223 0.0001 0>}]
 cape_04_r [{shapeType=Box, extentX=18.0686, extentY=1.4348, extentZ=31.6151, center=<9.223 0.0001 0>}]
 cape_05_l [{shapeType=Box, extentX=19.125, extentY=4.4375, extentZ=35.0724, center=<8.6769 0.0001 0>}]
 cape_05_r [{shapeType=Box, extentX=19.125, extentY=4.4375, extentZ=35.0724, center=<8.6769 0.0001 0>}]
 cape_06_l [{shapeType=Box, extentX=15.335, extentY=4.625, extentZ=36.6022, center=<6.1705 -0.0001 0>}]
 cape_06_r [{shapeType=Box, extentX=15.335, extentY=4.625, extentZ=36.6022, center=<6.1705 -0.0001 0>}]
 cape_07_l [{shapeType=Box, extentX=14.625, extentY=2.6875, extentZ=38.5406, center=<6.713 -0.0001 0>}]
 cape_07_r [{shapeType=Box, extentX=14.625, extentY=2.6875, extentZ=38.5406, center=<6.713 -0.0001 0>}]
 cape_08_l [{shapeType=Box, extentX=25.0625, extentY=2.6875, extentZ=39.8531, center=<12.5115 0.0002 0>}]
 cape_08_r [{shapeType=Box, extentX=25.0625, extentY=2.6875, extentZ=39.8531, center=<12.5115 0.0002 0>}]
 cape_mid_01_l [{shapeType=Capsule, length=1, radius=1, shapeName=cape_mid_01_l_capsule}]
 cape_mid_01_r [{shapeType=Capsule, length=1, radius=1, shapeName=cape_mid_01_r_capsule}]
 cape_mid_02_l [{shapeType=Box, extentX=22.5625, extentY=1.625, extentZ=29.0244, center=<11.5978 0.0001 0>}]
 cape_mid_02_r [{shapeType=Box, extentX=22.5625, extentY=1.625, extentZ=29.0244, center=<11.5978 0.0001 0>}]
 cape_mid_03_l [{shapeType=Box, extentX=18.0686, extentY=1.4348, extentZ=31.6151, center=<9.223 0.0001 0>}]
 cape_mid_03_r [{shapeType=Box, extentX=18.0686, extentY=1.4348, extentZ=31.6151, center=<9.223 0.0001 0>}]
 cape_mid_04_l [{shapeType=Box, extentX=19.125, extentY=4.4375, extentZ=35.0724, center=<8.6769 0.0001 0>}]
 cape_mid_04_r [{shapeType=Box, extentX=19.125, extentY=4.4375, extentZ=35.0724, center=<8.6769 0.0001 0>}]
 cape_mid_05_l [{shapeType=Box, extentX=15.335, extentY=4.625, extentZ=53.5265, center=<6.1705 -0.0001 0>}]
 cape_mid_05_r [{shapeType=Box, extentX=15.335, extentY=4.625, extentZ=53.5265, center=<6.1705 -0.0001 0>}]
 cape_mid_06_l [{shapeType=Box, extentX=14.625, extentY=2.6875, extentZ=55.4648, center=<6.713 -0.0001 0>}]
 cape_mid_06_r [{shapeType=Box, extentX=14.625, extentY=2.6875, extentZ=55.4648, center=<6.713 -0.0001 0>}]
 cape_mid_07_l [{shapeType=Box, extentX=25.0625, extentY=2.6875, extentZ=56.7773, center=<12.5115 0.0002 0>}]
 cape_mid_07_r [{shapeType=Box, extentX=25.0625, extentY=2.6875, extentZ=56.7773, center=<12.5115 0.0002 0>}]
 foot_l [{shapeType=Capsule, length=15.9367, radius=14.0633, shapeName=foot_l_capsule, center=<-3.8581 8.2738 0>, rotation=<rot 0 24.9997 -90>}]
 foot_r [{shapeType=Capsule, length=14.8216, radius=15.1784, shapeName=foot_r_capsule, center=<3.064 -8.4184 0>, rotation=<rot 0 19.9996 -90>}]
 head [{shapeType=Capsule, length=1, radius=1, shapeName=head_capsule}]
 neck_01 [{shapeType=Capsule, length=1, radius=1, shapeName=neck_01_capsule}]
 neck_02 [{shapeType=Capsule, length=1, radius=1, shapeName=neck_02_capsule}]
 pelvis [{shapeType=Capsule, length=13.8246, radius=17.9165, center=<-2.6215 -1.3695 0>}]
 spine_01 [{shapeType=Capsule, length=11.7115, radius=17.4407, center=<-2.6215 -0.7156 0>}]
 spine_02 [{shapeType=Capsule, length=0.02, radius=1.3612, shapeName=spine_02_capsule, rotation=<rot 0 -90 0>}]
 spine_03 [{shapeType=Capsule, length=5.0759, radius=14.8378, center=<-1.953 0.1364 0.0001>}]
 spine_04 [{shapeType=Capsule, length=14.6562, radius=13.6325, center=<2.1467 -1.2748 0.0002>}]
 spine_05 [
  {shapeType=Capsule, length=17.8437, radius=17.195, center=<-1.7514 2.4879 0>}
  {shapeType=Capsule, length=12.7812, radius=17.0625, center=<0.7875 3.1143 0>}
 ]
 thigh_l [{shapeType=Capsule, length=24.1721, radius=12.8966, shapeName=thigh_l_capsule, center=<-21.2797 0.0001 0>, rotation=<rot 90 0.0002 0.0006>}]
 thigh_r [{shapeType=Capsule, length=25.0504, radius=12.0184, shapeName=thigh_r_capsule, center=<20.4382 -0.0001 0>, rotation=<rot 90 160.5288 160.529>}]
}
constraints [
 {bone1Name=cape_04_l, bone2Name=cape_03_l, swing1LimitDegrees=5, swing2LimitDegrees=10, twistLimitDegrees=5}
 {bone1Name=cape_05_l, bone2Name=cape_04_l, swing1LimitDegrees=45, swing2LimitDegrees=45, twistLimitDegrees=5}
 {bone1Name=cape_06_l, bone2Name=cape_05_l, swing1LimitDegrees=45, swing2LimitDegrees=5, twistLimitDegrees=5}
 {bone1Name=cape_07_l, bone2Name=cape_06_l, swing1LimitDegrees=45, swing2LimitDegrees=5, twistLimitDegrees=5}
 {bone1Name=cape_08_l, bone2Name=cape_07_l, swing1LimitDegrees=5, swing2LimitDegrees=5, twistLimitDegrees=5}
 {bone1Name=cape_04_r, bone2Name=cape_03_r, swing1LimitDegrees=5, swing2LimitDegrees=10, twistLimitDegrees=5}
 {bone1Name=cape_05_r, bone2Name=cape_04_r, swing1LimitDegrees=45, swing2LimitDegrees=45, twistLimitDegrees=5}
 {bone1Name=cape_06_r, bone2Name=cape_05_r, swing1LimitDegrees=45, swing2LimitDegrees=5, twistLimitDegrees=5}
 {bone1Name=cape_07_r, bone2Name=cape_06_r, swing1LimitDegrees=45, swing2LimitDegrees=5, twistLimitDegrees=5}
 {bone1Name=cape_08_r, bone2Name=cape_07_r, swing1LimitDegrees=5, swing2LimitDegrees=5, twistLimitDegrees=5}
 {bone1Name=cape_mid_03_l, bone2Name=cape_mid_02_l, swing1LimitDegrees=5, swing2LimitDegrees=10, twistLimitDegrees=5}
 {bone1Name=cape_mid_04_l, bone2Name=cape_mid_03_l, swing1LimitDegrees=45, swing2LimitDegrees=45, twistLimitDegrees=5}
 {bone1Name=cape_mid_05_l, bone2Name=cape_mid_04_l, swing1LimitDegrees=45, swing2LimitDegrees=5, twistLimitDegrees=5}
 {bone1Name=cape_mid_06_l, bone2Name=cape_mid_05_l, swing1LimitDegrees=45, swing2LimitDegrees=5, twistLimitDegrees=5}
 {bone1Name=cape_mid_07_l, bone2Name=cape_mid_06_l, swing1LimitDegrees=5, swing2LimitDegrees=5, twistLimitDegrees=5}
 {bone1Name=cape_mid_03_r, bone2Name=cape_mid_02_r, swing1LimitDegrees=5, swing2LimitDegrees=10, twistLimitDegrees=5}
 {bone1Name=cape_mid_04_r, bone2Name=cape_mid_03_r, swing1LimitDegrees=45, swing2LimitDegrees=45, twistLimitDegrees=5}
 {bone1Name=cape_mid_05_r, bone2Name=cape_mid_04_r, swing1LimitDegrees=45, swing2LimitDegrees=5, twistLimitDegrees=5}
 {bone1Name=cape_mid_06_r, bone2Name=cape_mid_05_r, swing1LimitDegrees=45, swing2LimitDegrees=5, twistLimitDegrees=5}
 {bone1Name=cape_mid_07_r, bone2Name=cape_mid_06_r, swing1LimitDegrees=5, swing2LimitDegrees=5, twistLimitDegrees=5}
 {bone1Name=cape_03_l, bone2Name=cape_02_l, swing1LimitDegrees=0, swing2LimitDegrees=10, twistLimitDegrees=5}
 {bone1Name=cape_03_r, bone2Name=cape_02_r, swing1LimitDegrees=0, swing2LimitDegrees=10, twistLimitDegrees=5}
 {bone1Name=cape_mid_02_l, bone2Name=cape_mid_01_l, swing1LimitDegrees=0, swing2LimitDegrees=10, twistLimitDegrees=5}
 {bone1Name=cape_mid_02_r, bone2Name=cape_mid_01_r, swing1LimitDegrees=0, swing2LimitDegrees=10, twistLimitDegrees=5}
 {
  bone1Name=cape_02_l
  bone2Name=spine_05
  swing1LimitDegrees=45
  swing1Motion=Free
  swing2LimitDegrees=45
  swing2Motion=Free
  twistLimitDegrees=45
  twistMotion=Free
 }
 {
  bone1Name=cape_02_r
  bone2Name=spine_05
  swing1LimitDegrees=45
  swing1Motion=Free
  swing2LimitDegrees=45
  swing2Motion=Free
  twistLimitDegrees=45
  twistMotion=Free
 }
 {
  bone1Name=cape_mid_01_l
  bone2Name=spine_05
  swing1LimitDegrees=45
  swing1Motion=Free
  swing2LimitDegrees=45
  swing2Motion=Free
  twistLimitDegrees=45
  twistMotion=Free
 }
 {
  bone1Name=cape_mid_01_r
  bone2Name=spine_05
  swing1LimitDegrees=45
  swing1Motion=Free
  swing2LimitDegrees=45
  swing2Motion=Free
  twistLimitDegrees=45
  twistMotion=Free
 }
 {
  bone1Name=spine_01
  bone2Name=pelvis
  swing1LimitDegrees=45
  swing1Motion=Free
  swing2LimitDegrees=45
  swing2Motion=Free
  twistLimitDegrees=45
  twistMotion=Free
 }
 {
  bone1Name=spine_02
  bone2Name=spine_01
  swing1LimitDegrees=45
  swing1Motion=Free
  swing2LimitDegrees=45
  swing2Motion=Free
  twistLimitDegrees=45
  twistMotion=Free
 }
 {
  bone1Name=spine_03
  bone2Name=spine_02
  swing1LimitDegrees=45
  swing1Motion=Free
  swing2LimitDegrees=45
  swing2Motion=Free
  twistLimitDegrees=45
  twistMotion=Free
 }
 {
  bone1Name=spine_04
  bone2Name=spine_03
  swing1LimitDegrees=45
  swing1Motion=Free
  swing2LimitDegrees=45
  swing2Motion=Free
  twistLimitDegrees=45
  twistMotion=Free
 }
 {
  bone1Name=spine_05
  bone2Name=spine_04
  swing1LimitDegrees=45
  swing1Motion=Free
  swing2LimitDegrees=45
  swing2Motion=Free
  twistLimitDegrees=45
  twistMotion=Free
 }
 {
  bone1Name=neck_01
  bone2Name=spine_05
  swing1LimitDegrees=45
  swing1Motion=Free
  swing2LimitDegrees=45
  swing2Motion=Free
  twistLimitDegrees=45
  twistMotion=Free
 }
 {
  bone1Name=neck_02
  bone2Name=neck_01
  swing1LimitDegrees=45
  swing1Motion=Free
  swing2LimitDegrees=45
  swing2Motion=Free
  twistLimitDegrees=45
  twistMotion=Free
 }
 {
  bone1Name=head
  bone2Name=neck_02
  swing1LimitDegrees=45
  swing1Motion=Free
  swing2LimitDegrees=45
  swing2Motion=Free
  twistLimitDegrees=45
  twistMotion=Free
 }
 {
  bone1Name=cape_03_l
  bone2Name=cape_mid_02_l
  swing1LimitDegrees=45
  swing1Motion=Free
  swing2LimitDegrees=45
  swing2Motion=Free
  twistLimitDegrees=45
  twistMotion=Free
 }
 {
  bone1Name=cape_03_r
  bone2Name=cape_mid_02_r
  swing1LimitDegrees=45
  swing1Motion=Free
  swing2LimitDegrees=45
  swing2Motion=Free
  twistLimitDegrees=45
  twistMotion=Free
 }
]
